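; Front-end preview actor that displays an item wrap on sample meshes for the current locker subslot.
; Each loaded preview entry replaces the skeletal mesh component at the entry's transform with Y negated, applies the wrap, and sets its animation or anim class.
; A transition particle and brief hide cover the swap.
; Special actions load the next model or fake a weapon shot with wrap reaction, fire animation and muzzle flash.
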(event EventOnPreviewSkeletalMeshLoaded
  (params (LoadedMesh :type Skeletal%20Mesh%20Object%20Reference) (LoadedAnimationAsset :type Animation%20Asset%20Object%20Reference) (LoadedAnimInstanceClass :type Object%20Class%20Reference) (EntryInfo :type Item%20Wrap%20Preview%20Entry%20Structure%20%28by%20ref%29))
  (bind (_previewobject _previewobjectanimation _previewobjectaniminstance _previewtransform _bpreviewusingvehicleshader _previewobjectfiringanimation _previewobjectmuzzleflashparticlesystem) (Utilities|Struct|BreakItemWrapPreviewEntry EntryInfo))
  (bind (_previewtransform_location _previewtransform_rotation _previewtransform_scale) (Math|Transform|BreakTransform _previewtransform))
  (bind (_previewtransform_location_x _previewtransform_location_y _previewtransform_location_z) (Math|Vector|BreakVector _previewtransform_location))
  (bind _skelmeshcomp (Variables|Default|GetSkelMeshComp))
  (bind _skelmeshcomp_2 (Variables|Default|GetSkelMeshComp))
  (Utilities|FlowControl|Sequence
    (:then_0
      (Utilities|IsValid _skelmeshcomp
        (:Is%20Valid
          (Components|DestroyComponent _skelmeshcomp))))
    (:then_1
      (bind _returnvalue (AddComponent|NODE_AddSkeletalMeshComponent-0 :RelativeTransform_Location "0, 0, 0" :RelativeTransform_Rotation _previewtransform_rotation :RelativeTransform_Scale _previewtransform_scale))
      (bind _output_get (Variables|Default|SetSkelMeshComp _returnvalue))
      (Transformation|AddWorldOffset _output_get (Math|Vector|MakeVector _previewtransform_location_x (* -1 _previewtransform_location_y) _previewtransform_location_z) false false)
      (Variables|Default|SetWeaponFireAnim (Utilities|ResolveSoftReference _previewobjectfiringanimation))
      (Variables|Default|SetMuzzleFlashParticleSystem (Utilities|ResolveSoftReference _previewobjectmuzzleflashparticlesystem)))
    (:then_2
      (Components|SkinnedMesh|SetSkinnedAssetandUpdate _skelmeshcomp_2 LoadedMesh true)
      (CallFunction|ApplyWraptoSkelMesh _skelmeshcomp_2)
      (Utilities|IsValid LoadedAnimationAsset
        (:Is%20Valid
          (Components|Animation|OverrideAnimationData _skelmeshcomp_2 LoadedAnimationAsset true true 0.0 1.0))
        (:Is%20Not%20Valid
          (if (Utilities|IsValidClass LoadedAnimInstanceClass)
            (Components|SkeletalMesh|SetAnimInstanceClass _skelmeshcomp_2 LoadedAnimInstanceClass)
            (Variables|/Script/Engine.SkinnedMeshComponent|SetVisibilityBasedAnimTickOption "AlwaysTickPoseAndRefreshBones" _skelmeshcomp_2)))))
    (:then_3
      (Rendering|SetActorHiddenInGame true)
      (Effects|SpawnEmitteratLocation "/Game/Effects/Fort_Effects/UI/Transition/P_FrontEndCharacterTransition.P_FrontEndCharacterTransition" (Transformation|GetActorLocation) "0,90.000000, 0")
      (Utilities|FlowControl|Delay 0.1)
      (Rendering|SetActorHiddenInGame false))))

(event Emote|EventPreviewItemWrap
  (params (ItemWrap :type Athena%20Item%20Wrap%20Definition%20Object%20Reference))
  (bind _returnvalue (FrontEndContext|GetCurrentLockerSubslotIndex (PlayerController|LocalPlayerSubsystems|GetFortFrontEndContext (Game|GetPlayerController 0))))
  (CallFunction|StartLoadingNextPreviewActor _returnvalue))

(event Visuals|EventPerformSpecialAction
  (params (ActionName :type Name :default "None"))
  (bind _itemwrapmodifier (Variables|FortItemWrapPreviewActor|GetItemWrapModifier))
  (switch name ActionName
    (:SpecialPreviewAction_SwitchModel
      (bind _returnvalue (FrontEndContext|GetCurrentLockerSubslotIndex (PlayerController|LocalPlayerSubsystems|GetFortFrontEndContext (Game|GetPlayerController 0))))
      (CallFunction|StartLoadingNextPreviewActor _returnvalue))
    (:SpecialPreviewAction_TriggerReaction
      (Utilities|IsValid _itemwrapmodifier
        (:Is%20Valid
          (Class|CustomCosmeticModifierBase|OnWeaponFired _itemwrapmodifier)
          ;[1 1 1] Play kick animation in the frontend
          (bind _asanim_sequence_base (Utilities|Casting|CastToAnimSequenceBase (Variables|Default|GetWeaponFireAnim))
            (:then
              (Components|Animation|PlayAnimation (Variables|Default|GetSkelMeshComp) _asanim_sequence_base false))
            (:CastFailed
              _)) ; This should be a cast to an AnimationAsset, but apparently we can't do that in BPs for some reason
          (bind _asparticle_system (Utilities|Casting|CastToParticleSystem (Variables|Default|GetMuzzleFlashParticleSystem)))
          (Effects|SpawnEmitterAttached :EmitterTemplate _asparticle_system :AttachToComponent (Variables|Default|GetSkelMeshComp) :AttachPointName "Muzzle" :Scale "1.000000,1.000000,1.000000" :LocationType "SnapToTargetIncludingScale" :bAutoDestroy true :PoolingMethod "None" :bAutoActivate true))))))

(fn ConstructionScript)
